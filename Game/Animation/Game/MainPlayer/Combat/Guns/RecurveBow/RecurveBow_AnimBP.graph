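# Weapon AnimBlueprint for the recurve bow, posing the bow's own skeleton through aiming, charging and out-of-ammo.
# A state machine switches between an aim-offset Default pose and Charging, with normal and fast entry crossfades and an instant exit.
# Charging uses gender-specific blendspaces with curve-driven hammer rotation, blends to an aimed loop at max charge, and adds yaw and crouch-move additives.
# OutOfAmmoAlpha layers a no-ammo pose from arrow_nock, between DefaultSlot and ReloadSlot.
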AnimBlueprint /Game/Animation/Game/MainPlayer/Combat/Guns/RecurveBow/RecurveBow_AnimBP
target_skeleton=/Game/Weapons/FORT_Crossbows/Black_Metal/RecurveBow_Skeleton
graphs [
 AnimGraph {nodes [
   n1 = Sequence Player 'Recurve_Bow_NoAmmo_Pose_Weapon'() {sequence=&/Recurve_Bow_NoAmmo_Pose_Weapon}
   AnimGraphNode_StateMachine_0 = RecurveBow() {is=title AnimGraphNode_StateMachine}
   n2 = Slot 'DefaultSlot'(Source=@AnimGraphNode_StateMachine_0)
   n3 = Get OutOfAmmoAlpha() -> (OutOfAmmoAlpha:float)
   Output Pose(Result=Slot 'ReloadSlot'(Source=Layered blend per bone(BasePose=@n2, BlendPoses_0=@n1, BlendWeights_0=@n3) {layerSetup [{branchFilters [{boneName=arrow_nock}]}]}) {slotName=ReloadSlot}) {layerGroup=None}
  ]}
 RecurveBow {
  entry=Default
  path=AnimGraph.AnimGraphNode_StateMachine_0.RecurveBow
  ids=id
  nodes [
   AnimStateNode_0 = Default() {is=title AnimStateNode}
   AnimStateNode_1 = Charging() {is=title AnimStateNode}
   AnimStateTransitionNode_0 = AnimStateTransitionNode() {crossfadeDuration=0.55}
   AnimStateTransitionNode_1 = AnimStateTransitionNode() {crossfadeDuration=0.1}
   AnimStateTransitionNode_4 = AnimStateTransitionNode() {crossfadeDuration=0}
  ]
  links [
   {
    dst=Default
    rule=AnimGraph.AnimGraphNode_StateMachine_0.RecurveBow.AnimStateTransitionNode_4.AnimationTransitionGraph_0
    src=Charging
    via=AnimStateTransitionNode_4
   }
   {
    dst=Charging
    rule=AnimGraph.AnimGraphNode_StateMachine_0.RecurveBow.AnimStateTransitionNode_0.AnimationTransitionGraph_0
    src=Default
    via=AnimStateTransitionNode_0
   }
   {
    dst=Charging
    rule=AnimGraph.AnimGraphNode_StateMachine_0.RecurveBow.AnimStateTransitionNode_1.AnimationTransitionGraph_0
    src=Default
    via=AnimStateTransitionNode_1
   }
  ]
 }
 Default {
  path=AnimGraph.AnimGraphNode_StateMachine_0.RecurveBow.AnimStateNode_0.Default
  nodes [
   n1 = Get BowAimYaw() -> (BowAimYaw:float)
   n2 = Get BowAimPitch() -> (BowAimPitch:float)
   Blendspace Player 'RecurveBow_Targeted_AO_BS_Weapon'(X=@n1, Y=@n2) {blendSpace=/Game/Animation/Game/MainPlayer/Poses/Medium/Male/AO/RecurveBow/RecurveBow_Targeted_AO_BS_Weapon}
   n3 = Local Space Ref Pose()
   n4 = Sequence Player 'Recurve_Bow_BasePose_Weapon'() {sequence=&/Recurve_Bow_BasePose_Weapon}
   n5 = Get bShouldPlayDefaultPose() -> (bShouldPlayDefaultPose:bool)
   Output Animation Pose(Result=Apply Additive(
     Base=Blend Poses by bool(BlendPose_0=@n3, BlendPose_1=@n4, bActiveValue=@n5) {blendType=Linear}
     Additive=Blendspace Player 'RecurveBow_Targeted_AO_BS_Weapon'(X=@n1, Y=@n2) {blendSpace=&RecurveBow_Targeted_AO_BS_Weapon}
    ) {alphaBoolBlend {blendOption=Linear}})
  ]
 }
 Charging {
  path=AnimGraph.AnimGraphNode_StateMachine_0.RecurveBow.AnimStateNode_1.Charging
  nodes [
   n1 = Blendspace Player 'RecurveBow_Targeted_AO_BS_Weapon'(X=Get BowAimYaw(), Y=0) {blendSpace=&RecurveBow_Targeted_AO_BS_Weapon}
   n2 = Get BowAimPitch()
   n3 = Get ChargePlayRate() -> (ChargePlayRate:float)
   n4 = Get ChargeBlendSpaceCrouchSpeed() -> (ChargeBlendSpaceCrouchSpeed:float)
   n5 = "Blend Poses (EFortCustomGender)"!(
    BlendPose_0:Pose Link Structure=Blendspace Player 'Recurve_Bow_Combined_Charging_BS_Weapon'!(X=@n4, Y=@n2, PlayRate:float=@n3) -> (Pose) {bLoop=false, blendSpace=&/Recurve_Bow_Combined_Charging_BS_Weapon}
    BlendPose_1:Pose Link Structure=Blendspace Player 'Recurve_Bow_Combined_Charging_BS_Weapon_CMF'!(X=@n4, Y=@n2, PlayRate:float=@n3) -> (Pose) {bLoop=false, blendSpace=&/Recurve_Bow_Combined_Charging_BS_Weapon_CMF}
    BlendTime_0:float=0.1
    BlendTime_1:float=0.1
    ActiveEnumValue:Byte=Get OwnerGender() -> (OwnerGender:EFortCustomGender Enum=Invalid)
   ) -> (Pose:Pose Link Structure) {
    is=title AnimGraphNode_BlendListByEnum
    blendType=Linear
    blendPose [{linkId=-1, sourceLinkId=-1}, {linkId=-1, sourceLinkId=-1}]
    blendTime [0.1,0.1]
   }
   n6 = "Transform (Modify) Bone - Bone: hammer"(
    ComponentPose=Local To Component(LocalPose=@n5)
    Alpha=1
    Translation="0,0,0"
    Rotation=MakeRotator(Roll=GetCurveValue(CurveName=Hammer_Charging))
    Scale="1,1,1"
   ) {
    rotationMode=BMM_Additive
    rotationSpace=BCS_BoneSpace
    scaleMode=BMM_Additive
    translationMode=BMM_Additive
    alphaBoolBlend {blendOption=Linear}
    boneToModify {boneName=hammer}
   }
   n7 = "Blend Poses (EFortCustomGender)"(
    BlendPose_0=Sequence Player 'Recurve_Bow_ChargingLoop_CMM'() {sequence=&/Recurve_Bow_ChargingLoop_CMM}
    BlendPose_1=Sequence Player 'Recurve_Bow_ChargingLoop_CMF'() {sequence=&/Recurve_Bow_ChargingLoop_CMF}
    BlendTime_0=0.1
    BlendTime_1=0.1
    ActiveEnumValue=Get OwnerGender() -> (OwnerGender=Invalid)
   ) {blendType=Linear, blendPose [{linkId=-1, sourceLinkId=-1}, {linkId=-1, sourceLinkId=-1}], blendTime [0.1,0.1]}
   n8 = AimOffset Player 'Recurve_Bow_Charging_Loop_Weapon_AO'!(X:float=@n2, BasePose:Pose Link Structure=@n7, Alpha:float=1) -> (Pose:Pose Link Structure) {
    blendSpace=/Game/Animation/Game/MainPlayer/Poses/Medium/Male/AO/RecurveBow/Recurve_Bow_Charging_Loop_Weapon_AO
    alphaBoolBlend {blendOption=Linear}
   }
   n9 = Two Way Blend!(A=Component To Local(ComponentPose=@n6), B=@n8, bAlphaBoolEnabled:bool=Get bIsBowAtMaxCharge() -> (bIsBowAtMaxCharge:bool)) -> (Pose) {blendNode {alphaInputType=Bool, bResetChildOnActivation=true, alphaBoolBlend {blendInTime=0.5, blendOption=Linear}}}
   Output Animation Pose(Result=Apply Additive!(
     Base=Apply Additive!(Base=@n9, Additive=@n1, Alpha:float) -> (Pose) {alphaBoolBlend {blendOption=Linear}}
     Additive="Sequence Player (additive) 'Recurve_Bow_CrouchCharging_Into_CMM_Pose_Weapon'"() {sequence=&/Recurve_Bow_CrouchCharging_Into_CMM_Pose_Weapon}
     bAlphaBoolEnabled:bool=Get bIsCrouchMoving() -> (bIsCrouchMoving:bool)
    ) -> (Pose) {alphaInputType=Bool, alphaBoolBlend {blendInTime=0.45, blendOutTime=0.45}})
  ]
 }
 AnimationTransitionGraph_0 {
  path=AnimGraph.AnimGraphNode_StateMachine_0.RecurveBow.AnimStateTransitionNode_0.AnimationTransitionGraph_0
  nodes [Result(bCanEnterTransition=Get bEnterChargingNormalTransition() -> (bEnterChargingNormalTransition:bool))]
 }
 AnimationTransitionGraph_0 {
  path=AnimGraph.AnimGraphNode_StateMachine_0.RecurveBow.AnimStateTransitionNode_1.AnimationTransitionGraph_0
  nodes [Result(bCanEnterTransition=Get bEnterChargingFastTransition() -> (bEnterChargingFastTransition:bool))]
 }
 AnimationTransitionGraph_0 {
  path=AnimGraph.AnimGraphNode_StateMachine_0.RecurveBow.AnimStateTransitionNode_4.AnimationTransitionGraph_0
  nodes [Result(bCanEnterTransition=NOT Boolean(A=Get bIsBowCharging() -> (bIsBowCharging:bool)))]
 }
 EventGraph {nodes [
   TryGetPawnOwner()
   Event BlueprintUpdateAnimation()
  ]}
]
variables []
